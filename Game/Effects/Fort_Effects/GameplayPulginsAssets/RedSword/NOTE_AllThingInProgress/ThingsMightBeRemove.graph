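# An empty Material placeholder in the RedSword effects folder, under a work-in-progress path whose name marks it as possibly slated for removal.
# It has no expressions and uses the Default Lit shading model, so it defines no custom surface behavior.
# Automatic usage-flag setting in the editor is turned off.
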Material /Game/Effects/Fort_Effects/GameplayPulginsAssets/RedSword/NOTE_AllThingInProgress/ThingsMightBeRemove
graph {expressions {}, properties {bAutomaticallySetUsageInEditor=false, usedShadingModels=Default Lit}}
